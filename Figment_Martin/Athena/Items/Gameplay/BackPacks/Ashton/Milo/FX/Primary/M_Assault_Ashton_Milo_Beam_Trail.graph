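# Translucent unlit beam-trail material for the Ashton Milo backpack effect, drawing a scrolling blue energy streak.
# Several panning beam, portal-edge, laser and noise textures, plus a static ramp, blend into the streak pattern.
# That pattern drives a blue gradient with an HDR-bright center, and an alpha shaped by sphere masks along the trail's UVs.
# Particle color green sets center width, blue the lengthwise mask, and alpha the fade, with time-of-day scaling opacity.
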Material /Figment_Martin/Athena/Items/Gameplay/BackPacks/Ashton/Milo/FX/Primary/M_Assault_Ashton_Milo_Beam_Trail
graph {
 properties {bUsedWithBeamTrails=true, blendMode=BLEND_Translucent, pixelDepthOffsetMode=PDOM_Legacy, shadingModel=MSM_Unlit, usedShadingModels=Unlit}
 nodes [
  n1 = SphereMask(A=WorldPosition().XYZ, B=ObjectPositionWS(), Radius=Constant() {r=2200}, Hardness=Constant())
  n2 = TextureCoordinate() {coordinateIndex=1}
  n3 = TextureSample(UVs=Panner(Coordinate=TextureCoordinate() {coordinateIndex=1, uTiling=0.7, vTiling=0.55}) {speedX=-0.8}) {samplerType=SAMPLERTYPE_Normal, texture=/Figment_Martin/Effects/Fort_Effects/Textures/T_Weap_Distort_N_02}
  Add(
   A=AppendVector(A=Multiply(A=@n1) {constB=6}, B=ComponentMask(None=@n2) {g=true})
   B=Multiply(A=ComponentMask(None=@n3.RGB) {g=true, r=true}) {constB=0.12}
  )
  n4 = Panner(Coordinate=TextureCoordinate() {coordinateIndex=1, uTiling=5}) {speedX=3}
  n5 = Panner(Coordinate=TextureCoordinate() {coordinateIndex=1, uTiling=2}) {speedX=1.5}
  n6 = TextureSample(UVs=Panner(Coordinate=TextureCoordinate() {coordinateIndex=1, uTiling=8}) {speedX=1.9}) {samplerType=SAMPLERTYPE_LinearColor, texture=/Game/Effects/Fort_Effects/Gameplay/SmokeDustExplosions/Textures/T_NoisePack_LinearColor}
  n7 = LinearInterpolate(
   A=Multiply(A=TextureSample(UVs=@n4) {texture=/Game/Effects/Fort_Effects/Textures/StormModifiers/T_DamageReflectBeam}.G) {constB=1.2}
   B=Multiply(A=TextureSample(UVs=@n5) {texture=/Game/Effects/Fort_Effects/Textures/Taker/T_PortalEdge_02_Packed}.G, B=@n6.R)
   Alpha=Constant() {r=0.3}
  )
  n8 = TextureSample() {texture=/Figment_Martin/Effects/Fort_Effects/Textures/Ramps/T_Water_Top_Ramp}
  n9 = Panner(Coordinate=TextureCoordinate() {coordinateIndex=1, uTiling=4}) {speedX=2.3}
  n10 = Add(
   A=LinearInterpolate(A=@n7, B=Multiply(A=@n8.R) {constB=0.6}, Alpha=Constant() {r=0.55})
   B=Multiply(A=TextureSample(UVs=@n9) {texture=/Game/Effects/Fort_Effects/Textures/Tile/T_LaserTrail}.R, B=Constant() {r=0.7})
  )
  n11 = TextureSample(UVs=Panner(Coordinate=TextureCoordinate() {coordinateIndex=1, uTiling=0.7}) {speedX=1.5}) {samplerType=SAMPLERTYPE_LinearColor, texture=/Game/Effects/Fort_Effects/Textures/Tile/T_Raygun_Beam_01}
  Add(A=@n10, B=@n11.G)
  n12 = SphereMask(
   A=MaterialFunctionCall(Float2=TextureCoordinate() {coordinateIndex=1}) {materialFunction=&BreakOutFloat2Components}.G
   B=Constant() {r=0.5}
   Radius=ParticleColor().G
  ) {hardnessPercent=0}
  n13 = ComponentMask(None=TextureCoordinate()) {r=true}
  n14 = ScalarParameter() {defaultValue=0.125, parameterName=Range Contrast}
  n15 = LinearInterpolate(
   A=LinearInterpolate(A=Constant3Vector() {constant=<rgba 0.0171 0.0186 0.7 0>}, B=Constant3Vector() {constant=<rgba 0.0123 0.2009 0.55 0>}, Alpha=@n7)
   B=Constant3Vector() {constant=<rgba 2.0312 2.2334 6 0>}
   Alpha=Multiply(A=Multiply(A=@n12) {constB=0.7}, B=Power(Base=@n13, Exp=@n14))
  )
  Add(
   A=MaterialFunctionCall(LDR Input=Clamp(None=@n15), HDR Tint=@n15) {materialFunction=&DeriveHDRfromLDR}.RemappedRange
   B=Multiply(A=@n11.G, B=Constant3Vector() {constant=<rgba 0.1692 0.3061 0.58 0>})
  )
  Constant()
  Constant() {r=1}
  LinearInterpolate(
   A=Constant() {r=1}
   B=Constant() {r=-5}
   Alpha=MaterialFunctionCall() {materialFunction=&DebugTimeSine}.Result
  )
  n16 = TextureCoordinate() {coordinateIndex=1}
  n17 = ParticleColor()
  n18 = Multiply(
   A=@n10
   B=SphereMask(
    A=MaterialFunctionCall(Float2=@n16) {materialFunction=&BreakOutFloat2Components}.R
    B=@n17.B
    Radius=@n17.B
    Hardness=Constant() {r=0.5}
   )
  )
  n19 = ScalarParameter() {defaultValue=0.6}
  n20 = SphereMask(
   A=MaterialFunctionCall(Float2=TextureCoordinate() {coordinateIndex=1}) {materialFunction=&BreakOutFloat2Components}.R
   B=Constant() {r=0.6}
   Radius=@n19
  ) {hardnessPercent=0}
  n21 = Constant() {r=1}
  n22 = Constant() {r=0.65}
  Multiply(
   A=Multiply(A=Multiply(A=@n18, B=@n20), B=@n17.A)
   B=MaterialFunctionCall(Color_Morning=@n21, Color_Day=@n21, Color_Evening=@n22, Color_Night=@n22) {materialFunction=&Color_Time_of_Day}.Result
  )
  SphereMask(A=WorldPosition().XYZ, B=ObjectPositionWS(), Radius=@n19) {hardnessPercent=0}
 ]
}
