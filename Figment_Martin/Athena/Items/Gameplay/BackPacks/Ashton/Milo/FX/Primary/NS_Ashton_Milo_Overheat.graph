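# Niagara overheat effect for the Ashton Milo backpack, venting blue-tinted smoke and flashes from local +Y offsets.
# FlashSmoke spawns short-lived meshes in two spawn groups, using a visibility tag across two mesh renderers.
# MuzzleOverheat throws flipbook smoke sideways, flipping X velocity per spawn group, while GPU emitter MuzzleOverheat004 pushes larger, longer-lived puffs along +Y.
# The other five emitters, including streak and glow variants, are disabled.
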NiagaraSystem /Figment_Martin/Athena/Items/Gameplay/BackPacks/Ashton/Milo/FX/Primary/NS_Ashton_Milo_Overheat
dependencies {
 usedDataInterfaces [
  &NiagaraDataInterfaceCurve
  &NiagaraDataInterfaceArrayFloat
  &NiagaraDataInterfaceSpriteRendererInfo
  &NiagaraDataInterfaceVectorCurve
 ]
 usedDynamicInputs []
 usedModules [&SystemState]
}
emitters {
 FlashSmoke {
  emitterUpdateScript [
   &EmitterState()
   &SpawnBurst_Instantaneous(Use Spawn Probability=false, UseLoopCountLimit=false) {enabled=false}
   &SpawnRate(SpawnRate=7.5, Use Spawn Probability=false) {module=SpawnRate001}
   &SpawnRate(SpawnGroup=1, SpawnRate=7.5, Use Spawn Probability=false)
  ]
  particleSpawnScript [
   &FN_SetSpawnGroup(Number of Spawn Groups=2) {enabled=false}
   /SetVariables(Particles.SpawnGroup=&RandomRangeInt(Minimum=0, Maximum=1)) {enabled=false, isSetParameters=true}
   /SetVariables(Particles.MeshRotationEuler:Vector3f=&RandomRangeVector(Minimum=<0 0 0>, Maximum=<1 0 0>, Evaluation Type=::NewEnumerator1)) {enabled=false, isSetParameters=true}
   /SetVariables(Particles.MeshOrientation=/CascadeToNiagaraConverter/NiagaraScripts/DynamicInputScripts/MakeQuatFromEuler(Euler:Vector3f=$Particles.MeshRotationEuler, Scale:NiagaraFloat=0)) {enabled=false, isSetParameters=true}
   &InitializeParticle(
    Color=&MultiplyLinearColor(
     Linear Color=<rgba 0.4138 0.5518 1.4 1>
     Scale Factor=&MakeVector4_FromLinearColor(LinearColor=<rgba 0.2831 0.6105 1 1>)
    )
    Lifetime=0.5
    Mesh Scale=/Niagara/DynamicInputs/CascadeConversion/SplitVector(
     Double Component Vector Channels:ECascadeNiagaraTwoVectorChannels=::NewEnumerator1
     Double Component:NiagaraFloat=0.9
     Single Component:NiagaraFloat=0.67
    )
    Mesh Scale Mode=::NewEnumerator4
    UsePositionOffset=false
    Write Material Random=true
   )
   &Color(Color=&MakeLinearColorFromVectorAndFloat(
     "Vector (RGB)"=<0.4138 0.5518 1.4>
     "Float (Alpha)"=&FloatFromCurve(FloatCurve=DataInterface() {propertyValues {Curve {keys [<at 0 0.85>,<at 1 0.45>]}}})
    )) {enabled=false}
   /CascadeToNiagaraConverter/NiagaraScripts/ModuleScripts/AddVectorToPosition(Vector:Vector3f=&TransformVector(Vector=<2 42 -2>, Destination Space=::Simulation))
   &InitialMeshOrientation(
    Mesh Orientation Mode=::NewEnumerator7
    Rotation=&RandomRangeVector(Minimum=<0 0 0>, Maximum=<1 0 0>)
    Use Rotation Vector=true
   )
   /SetVariables(Particles.VisibilityTag=&MaskIntBySpawnGroup(Masked Int=0, Passthrough Int=1, Spawn Group Mask=0)) {isSetParameters=true}
  ]
  particleUpdateScript [
   &ParticleState()
   &ScaleMeshSize(Scale Factor=&VectorFromFloat(Value=&FloatFromCurve(FloatCurve=DataInterface() {propertyValues {Curve {keys [
          {arriveTangent=2.5199, interpMode=RCIM_Cubic, leaveTangent=2.5199, tangentMode=RCTM_User, time=0, value=0}
          {arriveTangent=0.508, interpMode=RCIM_Cubic, leaveTangent=0.508, time=0.158, value=0.4128}
          <at 1 0.5079>
         ]}}})))
   &DynamicMaterialParameters(Module.Param0WriteEnabled=true, Module.Param1WriteEnabled=true, Module.Param2WriteEnabled=true, Module.Param3WriteEnabled=true)
   &ScaleColor(
    Scale Alpha=&FloatFromCurve(FloatCurve=DataInterface() {propertyValues {Curve {keys [{arriveTangent=-2.383, interpMode=RCIM_Cubic, leaveTangent=-2.383, tangentMode=RCTM_User, time=0, value=1}, <at 1 0>]}}})
    Scale RGB=&VectorFromCurve(VectorCurve=DataInterface() {propertyValues {XCurve {keys [<at 0.7 8>,<at 1 1>]}, YCurve {keys [<at 0.7 8>,<at 1 1>]}, ZCurve {keys [<at 0.7 8>,<at 1 1>]}}})
    ScaleA=true
    ScaleRGB=true
   )
  ]
  rendererClasses [&NiagaraMeshRendererProperties]
  renderers [
   {rendererClass=&NiagaraMeshRendererProperties, rendererIndex=0}
   {rendererClass=&NiagaraMeshRendererProperties, rendererIndex=1}
  ]
 }
 FlashSmoke001 {
  bEnabled=false
  emitterUpdateScript [
   &EmitterState(Life Cycle Mode=::NewEnumerator1, Loop Behavior=::NewEnumerator1, Loop Duration=1.6, UseLoopDelay=true)
   &SpawnRate(SpawnRate=10, Use Spawn Probability=false)
  ]
  particleSpawnScript [
   /SetVariables(Particles.MeshRotationEuler:Vector3f=&RandomRangeVector(Minimum=<0 0.5 0>, Maximum=<1 0.5 0>, Evaluation Type=::NewEnumerator1)) {isSetParameters=true}
   /SetVariables(Particles.MeshOrientation=&MakeQuatFromEuler(Euler:Vector3f=$Particles.MeshRotationEuler, Scale:NiagaraFloat=360)) {isSetParameters=true}
   &InitializeParticle(
    Lifetime=0.2
    Mesh Scale=&SplitVector(
     Double Component Vector Channels:ECascadeNiagaraTwoVectorChannels=::NewEnumerator1
     Double Component:NiagaraFloat=0.9
     Single Component:NiagaraFloat=0.67
    )
    Mesh Scale Mode=::NewEnumerator4
    UsePositionOffset=false
    Write Material Random=true
   )
   &Color(Color=&MakeLinearColorFromVectorAndFloat(
     "Vector (RGB)"=<0.4138 0.5518 1.4>
     "Float (Alpha)"=&FloatFromCurve(FloatCurve=DataInterface() {propertyValues {Curve {keys [<at 0 0.85>,<at 1 0.45>]}}})
    ))
   &AddVectorToPosition(Vector:Vector3f=&TransformVector(Vector=<0 42 -2>, Destination Space=::Simulation))
  ]
  particleUpdateScript [
   &ParticleState()
   &ScaleMeshSize(Scale Factor=&VectorFromFloat(Value=&FloatFromCurve(FloatCurve=DataInterface() {propertyValues {Curve {keys [
          {arriveTangent=2.5199, interpMode=RCIM_Cubic, leaveTangent=2.5199, tangentMode=RCTM_User, time=0, value=0}
          {arriveTangent=0.508, interpMode=RCIM_Cubic, leaveTangent=0.508, time=0.158, value=0.4128}
          <at 1 0.5079>
         ]}}})))
   &DynamicMaterialParameters(Module.Param0WriteEnabled=true, Module.Param1WriteEnabled=true, Module.Param2WriteEnabled=true, Module.Param3WriteEnabled=true)
   &ScaleColor(
    Scale Alpha=&FloatFromCurve(FloatCurve=DataInterface() {propertyValues {Curve {keys [{arriveTangent=-2.383, interpMode=RCIM_Cubic, leaveTangent=-2.383, tangentMode=RCTM_User, time=0, value=1}, <at 1 0>]}}})
    Scale RGB=&VectorFromCurve(VectorCurve=DataInterface() {propertyValues {XCurve {keys [<at 0.7 8>,<at 1 1>]}, YCurve {keys [<at 0.7 8>,<at 1 1>]}, ZCurve {keys [<at 0.7 8>,<at 1 1>]}}})
    ScaleA=true
    ScaleRGB=true
   )
  ]
  rendererClasses [&NiagaraMeshRendererProperties]
  renderers [{rendererClass=&NiagaraMeshRendererProperties, rendererIndex=0}]
 }
 MuzzleOverheat {
  emitterUpdateScript [
   &EmitterState()
   &SpawnRate(
    SpawnRate=&Multiply_Float(
     A=10
     B=&FloatFromCurve(FloatCurve=DataInterface() {propertyValues {Curve {keys [<at 1 1>,<at 2 0.7>,<at 3 0>]}}}, CurveIndex=$Emitter.NormalizedLoopAge)
    )
    Use Spawn Probability=false
   ) {module=SpawnRate001}
   &SpawnRate(
    SpawnGroup=1
    SpawnRate=&Multiply_Float(
     A=10
     B=&FloatFromCurve(FloatCurve=DataInterface() {propertyValues {Curve {keys [<at 1 1>,<at 2 0.7>,<at 3 0>]}}}, CurveIndex=$Emitter.NormalizedLoopAge)
    )
    Use Spawn Probability=false
   )
  ]
  particleSpawnScript [
   /SetVariables(Particles.SpriteRotationRate=&RandomRangeFloat(Minimum=0, Maximum=0.2, Evaluation Type=::NewEnumerator1)) {isSetParameters=true}
   &InitializeParticle(
    Lifetime=&RandomRangeFloat(Minimum=0.35, Maximum=0.45, Evaluation Type=::NewEnumerator1)
    Sprite Rotation Angle=&RandomRangeFloat(Minimum=0, Maximum=1, Evaluation Type=::NewEnumerator1)
    Sprite Rotation Mode=::NewEnumerator1
    Sprite Size=&Vector2DFromFloat(Value=&RandomRangeFloat(Minimum=36, Maximum=48, Evaluation Type=::NewEnumerator1))
    Sprite Size Mode=::NewEnumerator4
    UsePositionOffset=false
    Write Material Random=true
   )
   &AddVelocity(Velocity=&MakeVector(
     X=&Multiply_Float(
      A=&SelectFloatFromArray(Direct Array Index=&SpawnGroup(), Float Selection Array=DataInterface() {propertyValues {FloatData [1,-1]}})
      B=&RandomRangeFloat(Minimum=360, Maximum=460)
     )
     Y=&RandomRangeFloat(Minimum=25, Maximum=100)
     Z=&RandomRangeFloat(Minimum=-6, Maximum=6)
    ))
   &AddVelocity(Velocity=&Multiply_VectorByFloat(
     Vector=&RandomRangeVector(Minimum=<360 25 -6>, Maximum=<445 100 7>, Evaluation Type=::NewEnumerator1)
     Float=1
    )) {enabled=false, module=AddVelocity001}
   &Color(Color=&MakeLinearColorFromVectorAndFloat(
     "Vector (RGB)"=<0.5431 0.6714 1>
     "Float (Alpha)"=&RandomRangeFloat(Minimum=0.175, Maximum=0.23, Evaluation Type=::NewEnumerator1)
    ))
   &InheritVelocity(Inherited Velocity Amount Scale=<0.75 0.75 0.75>, Inherited Velocity Speed Limit=0, Speed Limit=false, UseSpeedThreshold=false)
   &AddVectorToPosition(Vector:Vector3f=&TransformVector(Vector=<0 38 -2>, Destination Space=::Simulation))
  ]
  particleUpdateScript [
   &ParticleState()
   &ScaleSpriteSize(
    Scale Sprite Size Mode=::NewEnumerator0
    Uniform Scale Factor=&FloatFromCurve(FloatCurve=DataInterface() {propertyValues {Curve {keys [
         {arriveTangent=3.1664, interpMode=RCIM_Cubic, leaveTangent=3.1664, tangentMode=RCTM_User, time=0, value=0.1}
         {arriveTangent=0.7504, interpMode=RCIM_Cubic, leaveTangent=0.7504, tangentMode=RCTM_User, time=0.3843, value=1}
         {arriveTangent=0.438, interpMode=RCIM_Cubic, leaveTangent=0.438, tangentMode=RCTM_User, time=0.9987, value=1.0276}
        ]}}})
   )
   &SpriteRotationRate(Rotation Rate=&AngleConversion(
     Angle:NiagaraFloat=$Particles.SpriteRotationRate
     Angle Input:ENiagara_AngleInput=::NewEnumerator1
     Angle Output:ENiagara_AngleInput=::NewEnumerator0
    ))
   &Drag(
    Drag=&FloatFromCurve(FloatCurve=DataInterface() {propertyValues {Curve {keys [<at 0 4.4>,<at 1 10>]}}})
    Use Linear Drag=true
    Use Rotational Drag=true
   )
   &SolveForcesAndVelocity(Clamp Velocity=false, Limit Acceleration=false)
   &SubUVAnimation(
    End Frame Range Override=62
    "Sprite Renderer "=DataInterface() {propertyValues {SpriteRenderer=&self:MuzzleOverheat_0.NiagaraSpriteRendererProperties_0}}
    Start Frame Range Override=24
    UseEndFrame=true
    UseStartFrame=true
   )
   &ScaleColor(
    Scale Alpha=&FloatFromCurve(FloatCurve=DataInterface() {propertyValues {Curve {keys [
         {arriveTangent=10.221, interpMode=RCIM_Cubic, leaveTangent=10.221, tangentMode=RCTM_User, time=0.0059, value=0.005}
         {arriveTangent=1.528, interpMode=RCIM_Cubic, leaveTangent=1.528, tangentMode=RCTM_User, time=0.1611, value=0.9421}
         {arriveTangent=-1.8753, interpMode=RCIM_Cubic, leaveTangent=-1.8753, tangentMode=RCTM_User, time=0.5565, value=0.3858}
         {arriveTangent=-0.204, interpMode=RCIM_Cubic, leaveTangent=-0.204, tangentMode=RCTM_User, time=1, value=0}
        ]}}})
    Scale RGB=&VectorFromCurve(VectorCurve=DataInterface() {propertyValues {
       XCurve {keys [<at 0 0.0715 RCIM_Cubic>,<at 0.58 0.4825 RCIM_Cubic>]}
       YCurve {keys [<at 0 0.3317 RCIM_Cubic>,<at 0.58 0.5566 RCIM_Cubic>]}
       ZCurve {keys [<at 0 1 RCIM_Cubic>,<at 0.58 0.7 RCIM_Cubic>]}
      }})
    ScaleA=true
    ScaleRGB=true
   )
  ]
  rendererClasses [&NiagaraSpriteRendererProperties]
  renderers [{rendererClass=&NiagaraSpriteRendererProperties}]
 }
 MuzzleOverheat001 {
  bEnabled=false
  simTarget=GPUComputeSim
  emitterUpdateScript [
   &EmitterState()
   &SpawnRate(SpawnRate=100, Use Spawn Probability=false) {module=SpawnRate001}
   &SpawnRate(SpawnRate=100, Use Spawn Probability=false) {enabled=false}
  ]
  particleSpawnScript [
   &InitializeParticle(
    Lifetime=&RandomRangeFloat(Minimum=0.025, Maximum=0.1, Evaluation Type=::NewEnumerator1)
    Sprite Size=<1 3>
    Sprite Size Mode=::NewEnumerator4
    UsePositionOffset=false
    Write Material Random=true
   )
   &AddVelocity(
    Velocity=&RandomRangeVector(Minimum=<100 -25 -25>, Maximum=<300 25 25>, Evaluation Type=::NewEnumerator1)
    Velocity Speed Scale=&UniformAOrBFloat(A=-1, Evaluation Type=::NewEnumerator0, Recalculate Random Each Loop=true, Override Seed=false)
   )
   &Color(Color=&MakeLinearColorFromVectorAndFloat("Vector (RGB)"=<1 1 1>, "Float (Alpha)"=0.2))
   &AddVectorToPosition(Vector:Vector3f=&TransformVector(Vector=<0 43 -2>, Destination Space=::Simulation))
  ]
  particleUpdateScript [
   &ParticleState()
   &SolveForcesAndVelocity(Clamp Velocity=false, Limit Acceleration=false)
   &ScaleSpriteSizeBySpeed(Max Scale Factor=<10 50>, Source Velocity:Vector3f=&Multiply_Vector(A=<0 0.02 1>, B=$Particles.Velocity))
  ]
  rendererClasses [&NiagaraSpriteRendererProperties]
  renderers [{rendererClass=&NiagaraSpriteRendererProperties}]
 }
 MuzzleOverheat002 {
  bEnabled=false
  simTarget=GPUComputeSim
  emitterUpdateScript [
   &EmitterState(Life Cycle Mode=::NewEnumerator1, Loop Behavior=::NewEnumerator1, Loop Duration=1.8, UseLoopDelay=true)
   &SpawnBurst_Instantaneous(Spawn Count=4, Use Spawn Probability=false, UseLoopCountLimit=false)
   &SpawnRate(
    SpawnRate=&Multiply_Float(
     A=25
     B=&FloatFromCurve(FloatCurve=DataInterface() {propertyValues {Curve {keys [<at 1 1>,<at 2 0.7>,<at 3 0>]}}}, CurveIndex=$Emitter.NormalizedLoopAge)
    )
    Use Spawn Probability=false
   )
   &SpawnPerUnit(
    Max Movement Threshold=1500
    Module.Use Max Movement Threshold=true
    Module.Use Movement Tolerance=true
    Movement Tolerance=0.1
    Spawn Spacing:NiagaraFloat=&Divide_Float(A=1, B=&Divide_Float(A=1, B=10))
    Use Spawn Probability=false
   )
  ]
  particleSpawnScript [
   /SetVariables(Particles.SpriteRotationRate=&RandomRangeFloat(Minimum=0, Maximum=0.2, Evaluation Type=::NewEnumerator1)) {isSetParameters=true}
   &InitializeParticle(
    Lifetime=&RandomRangeFloat(Minimum=0.35, Maximum=0.45, Evaluation Type=::NewEnumerator1)
    Sprite Rotation Angle=&RandomRangeFloat(Minimum=0, Maximum=1, Evaluation Type=::NewEnumerator1)
    Sprite Rotation Mode=::NewEnumerator1
    Sprite Size=&Vector2DFromFloat(Value=&RandomRangeFloat(Minimum=24, Maximum=30, Evaluation Type=::NewEnumerator1))
    Sprite Size Mode=::NewEnumerator4
    UsePositionOffset=false
    Write Material Random=true
   )
   &AddVelocity(Velocity=&RandomRangeVector(Minimum=<-445 25 -6>, Maximum=<-360 100 7>, Evaluation Type=::NewEnumerator1))
   &Color(Color=&MakeLinearColorFromVectorAndFloat(
     "Vector (RGB)"=<0.5431 0.6714 1>
     "Float (Alpha)"=&RandomRangeFloat(Minimum=0.175, Maximum=0.23, Evaluation Type=::NewEnumerator1)
    ))
   &AddVectorToPosition(Vector:Vector3f=&TransformVector(
     Vector=&RandomRangeVector(Minimum=<0 38 0>, Maximum=<0 48 0>, Evaluation Type=::NewEnumerator1)
     Destination Space=::Simulation
    ))
   &InheritVelocity(Inherited Velocity Amount Scale=<0.75 0.75 0.75>, Inherited Velocity Speed Limit=0, Speed Limit=false, UseSpeedThreshold=false)
  ]
  particleUpdateScript [
   &ParticleState()
   &ScaleSpriteSize(
    Scale Sprite Size Mode=::NewEnumerator0
    Uniform Scale Factor=&FloatFromCurve(FloatCurve=DataInterface() {propertyValues {Curve {keys [
         {arriveTangent=3.1664, interpMode=RCIM_Cubic, leaveTangent=3.1664, tangentMode=RCTM_User, time=0.0019, value=0.182}
         {arriveTangent=0.7504, interpMode=RCIM_Cubic, leaveTangent=0.7504, tangentMode=RCTM_User, time=0.3768, value=1.1972}
         {arriveTangent=0.438, interpMode=RCIM_Cubic, leaveTangent=0.438, tangentMode=RCTM_User, time=0.9987, value=1.0276}
        ]}}})
   )
   &SpriteRotationRate(Rotation Rate=&AngleConversion(
     Angle:NiagaraFloat=$Particles.SpriteRotationRate
     Angle Input:ENiagara_AngleInput=::NewEnumerator1
     Angle Output:ENiagara_AngleInput=::NewEnumerator0
    ))
   &Drag(
    Drag=&FloatFromCurve(FloatCurve=DataInterface() {propertyValues {Curve {keys [<at 0 4.4>,<at 1 10>]}}})
    Use Linear Drag=true
    Use Rotational Drag=true
   )
   &SolveForcesAndVelocity(Clamp Velocity=false, Limit Acceleration=false)
   &SubUVAnimation(End Frame Range Override=62, Start Frame Range Override=24, UseEndFrame=false, UseStartFrame=false)
   &ScaleColor(
    Scale Alpha=&FloatFromCurve(FloatCurve=DataInterface() {propertyValues {Curve {keys [
         {arriveTangent=10.221, interpMode=RCIM_Cubic, leaveTangent=10.221, tangentMode=RCTM_User, time=0.0059, value=0.005}
         {arriveTangent=1.528, interpMode=RCIM_Cubic, leaveTangent=1.528, tangentMode=RCTM_User, time=0.1611, value=0.9421}
         {arriveTangent=-1.8753, interpMode=RCIM_Cubic, leaveTangent=-1.8753, tangentMode=RCTM_User, time=0.5565, value=0.3858}
         {arriveTangent=-0.204, interpMode=RCIM_Cubic, leaveTangent=-0.204, tangentMode=RCTM_User, time=1, value=0}
        ]}}})
    Scale RGB=&VectorFromCurve(VectorCurve=DataInterface() {propertyValues {
       XCurve {keys [<at 0 0.0715 RCIM_Cubic>,<at 0.58 0.4825 RCIM_Cubic>]}
       YCurve {keys [<at 0 0.3317 RCIM_Cubic>,<at 0.58 0.5566 RCIM_Cubic>]}
       ZCurve {keys [<at 0 1 RCIM_Cubic>,<at 0.58 0.7 RCIM_Cubic>]}
      }})
    ScaleA=true
    ScaleRGB=true
   )
  ]
  rendererClasses [&NiagaraSpriteRendererProperties]
  renderers [{rendererClass=&NiagaraSpriteRendererProperties}]
 }
 MuzzleOverheat003 {
  bEnabled=false
  simTarget=GPUComputeSim
  emitterUpdateScript [
   &EmitterState(Life Cycle Mode=::NewEnumerator1, Loop Behavior=::NewEnumerator1, Loop Duration=1.8, UseLoopDelay=true)
   &SpawnRate(SpawnRate=100, Use Spawn Probability=false)
  ]
  particleSpawnScript [
   &InitializeParticle(
    Lifetime=&RandomRangeFloat(Minimum=0.025, Maximum=0.1, Evaluation Type=::NewEnumerator1)
    Sprite Size=<1 3>
    Sprite Size Mode=::NewEnumerator4
    UsePositionOffset=false
    Write Material Random=true
   )
   &AddVelocity(Velocity=&RandomRangeVector(Minimum=<-100 -25 -25>, Maximum=<-300 25 25>, Evaluation Type=::NewEnumerator1))
   &Color(Color=&MakeLinearColorFromVectorAndFloat("Vector (RGB)"=<1 1 1>, "Float (Alpha)"=0.2))
   &AddVectorToPosition(Vector:Vector3f=&TransformVector(Vector=<0 43 -2>, Destination Space=::Simulation))
  ]
  particleUpdateScript [
   &ParticleState()
   &SolveForcesAndVelocity(Clamp Velocity=false, Limit Acceleration=false)
   &ScaleSpriteSizeBySpeed(Max Scale Factor=<10 50>, Source Velocity:Vector3f=&Multiply_Vector(A=<0 0.02 1>, B=$Particles.Velocity))
  ]
  rendererClasses [&NiagaraSpriteRendererProperties]
  renderers [{rendererClass=&NiagaraSpriteRendererProperties}]
 }
 MuzzleOverheat004 {
  simTarget=GPUComputeSim
  emitterUpdateScript [
   &EmitterState()
   &SpawnBurst_Instantaneous(Spawn Count=5, Use Spawn Probability=false, UseLoopCountLimit=false) {enabled=false}
   &SpawnRate(
    SpawnRate=&Multiply_Float(
     A=28
     B=&FloatFromCurve(FloatCurve=DataInterface() {propertyValues {Curve {keys [<at 1 1>,<at 2 0.7>,<at 3 0>]}}}, CurveIndex=$Emitter.NormalizedLoopAge)
    )
    Use Spawn Probability=false
   ) {enabled=false}
   &SpawnRate(
    SpawnRate=&Multiply_Float(
     A=8
     B=&FloatFromCurve(FloatCurve=DataInterface() {propertyValues {Curve {keys [<at 1 1>,<at 2 0.7>,<at 3 0>]}}}, CurveIndex=$Emitter.NormalizedLoopAge)
    )
    Use Spawn Probability=false
   ) {module=SpawnRate001}
   &SpawnPerUnit(
    Max Movement Threshold=1500
    Module.Use Max Movement Threshold=true
    Module.Use Movement Tolerance=true
    Movement Tolerance=0.1
    Spawn Spacing:NiagaraFloat=&Divide_Float(A=1, B=&Divide_Float(A=1, B=10))
    Use Spawn Probability=false
   ) {enabled=false}
  ]
  particleSpawnScript [
   /SetVariables(Particles.SpriteRotationRate=&RandomRangeFloat(Minimum=0, Maximum=0.2, Evaluation Type=::NewEnumerator1)) {isSetParameters=true}
   &InitializeParticle(
    Lifetime=&RandomRangeFloat(Minimum=0.75, Maximum=0.94, Evaluation Type=::NewEnumerator1)
    Sprite Rotation Angle=&RandomRangeFloat(Minimum=0, Maximum=1, Evaluation Type=::NewEnumerator1)
    Sprite Rotation Mode=::NewEnumerator1
    Sprite Size=&Multiply_Vector2D_ByFloat(
     Float=1
     Vector2D=&Vector2DFromFloat(Value=&RandomRangeFloat(Minimum=42, Maximum=52, Evaluation Type=::NewEnumerator1))
    )
    Sprite Size Mode=::NewEnumerator4
    UsePositionOffset=false
    Write Material Random=true
   )
   &AddVelocity(Velocity=&RandomRangeVector(Minimum=<-150 240 -2>, Maximum=<145 280 3>, Evaluation Type=::NewEnumerator1))
   &Color(Color=&MakeLinearColorFromVectorAndFloat(
     "Vector (RGB)"=<0.5431 0.6714 1>
     "Float (Alpha)"=&RandomRangeFloat(Minimum=0.175, Maximum=0.23, Evaluation Type=::NewEnumerator1)
    ))
   &AddVectorToPosition(Vector:Vector3f=&TransformVector(Vector=<0 30 0>, Destination Space=::Simulation))
   &InheritVelocity(Inherited Velocity Amount Scale=<0.75 0.75 0.75>, Inherited Velocity Speed Limit=0, Speed Limit=false, UseSpeedThreshold=false)
  ]
  particleUpdateScript [
   &ParticleState()
   &ScaleSpriteSize(
    Scale Sprite Size Mode=::NewEnumerator0
    Uniform Scale Factor=&FloatFromCurve(FloatCurve=DataInterface() {propertyValues {Curve {keys [
         {arriveTangent=3.1664, interpMode=RCIM_Cubic, leaveTangent=3.1664, tangentMode=RCTM_User, time=0.0019, value=0.3}
         {arriveTangent=0.7504, interpMode=RCIM_Cubic, leaveTangent=0.7504, tangentMode=RCTM_User, time=0.3768, value=1.1972}
         {arriveTangent=0.438, interpMode=RCIM_Cubic, leaveTangent=0.438, tangentMode=RCTM_User, time=0.9987, value=1.0276}
        ]}}})
   )
   &SpriteRotationRate(Rotation Rate=&AngleConversion(
     Angle:NiagaraFloat=$Particles.SpriteRotationRate
     Angle Input:ENiagara_AngleInput=::NewEnumerator1
     Angle Output:ENiagara_AngleInput=::NewEnumerator0
    ))
   &Drag(
    Drag=&FloatFromCurve(FloatCurve=DataInterface() {propertyValues {Curve {keys [<at 0 4.4>,<at 1 10>]}}})
    Use Linear Drag=true
    Use Rotational Drag=true
   )
   &SolveForcesAndVelocity(Clamp Velocity=false, Limit Acceleration=false)
   &SubUVAnimation(
    End Frame Range Override=62
    "Sprite Renderer "=DataInterface() {propertyValues {SpriteRenderer=&self:MuzzleOverheat004_0.NiagaraSpriteRendererProperties_4}}
    Start Frame Range Override=26
    UseEndFrame=true
    UseStartFrame=true
   )
   &ScaleColor(
    Scale Alpha=&FloatFromCurve(FloatCurve=DataInterface() {propertyValues {Curve {keys [
         {arriveTangent=10.221, interpMode=RCIM_Cubic, leaveTangent=10.221, tangentMode=RCTM_User, time=0.0059, value=0.005}
         {arriveTangent=1.528, interpMode=RCIM_Cubic, leaveTangent=1.528, tangentMode=RCTM_User, time=0.1611, value=0.9421}
         {arriveTangent=-1.8753, interpMode=RCIM_Cubic, leaveTangent=-1.8753, tangentMode=RCTM_User, time=0.5565, value=0.3858}
         {arriveTangent=-0.204, interpMode=RCIM_Cubic, leaveTangent=-0.204, tangentMode=RCTM_User, time=1, value=0}
        ]}}})
    Scale RGB=&VectorFromCurve(VectorCurve=DataInterface() {propertyValues {
       XCurve {keys [<at 0 0.0715 RCIM_Cubic>,<at 0.58 0.4825 RCIM_Cubic>]}
       YCurve {keys [<at 0 0.3317 RCIM_Cubic>,<at 0.58 0.5566 RCIM_Cubic>]}
       ZCurve {keys [<at 0 1 RCIM_Cubic>,<at 0.58 0.7 RCIM_Cubic>]}
      }})
    ScaleA=true
    ScaleRGB=true
   )
  ]
  rendererClasses [&NiagaraSpriteRendererProperties]
  renderers [{rendererClass=&NiagaraSpriteRendererProperties}]
 }
 MuzzleOverheat005 {
  bEnabled=false
  emitterUpdateScript [
   &EmitterState()
   &SpawnBurst_Instantaneous(Spawn Count=4, Use Spawn Probability=false, UseLoopCountLimit=false)
   &SpawnRate(SpawnRate=7, Use Spawn Probability=false)
  ]
  particleSpawnScript [
   &InitializeParticle(
    Lifetime=&RandomRangeFloat(Minimum=0.7, Maximum=0.8, Evaluation Type=::NewEnumerator1)
    Sprite Size=&Vector2DFromFloat(Value=&RandomRangeFloat(Minimum=18.5, Maximum=32, Evaluation Type=::NewEnumerator1))
    Sprite Size Mode=::NewEnumerator4
    UsePositionOffset=false
    Write Material Random=true
   )
   &Color(Color=&MakeLinearColorFromVectorAndFloat("Vector (RGB)"=<0 0.5857 2.1>, "Float (Alpha)"=1.8))
   &AddVectorToPosition(Vector:Vector3f=&TransformVector(
     Vector=&RandomRangeVector(Minimum=<-2 35 -2>, Maximum=<3 55 2>, Evaluation Type=::NewEnumerator1)
     Destination Space=::Simulation
    ))
   &AddVelocity(Velocity=&RandomRangeVector(Minimum=<-6 15 -8>, Maximum=<6 15 7>, Evaluation Type=::NewEnumerator1))
  ]
  particleUpdateScript [
   &ParticleState()
   &ScaleColor(
    Scale Alpha=&FloatFromCurve(FloatCurve=DataInterface() {propertyValues {Curve {keys [
         {arriveTangent=0.1908, time=0, value=0}
         {arriveTangent=-0.0391, leaveTangent=-0.0391, time=0.3035, value=0.1908}
         {arriveTangent=-0.274, interpMode=RCIM_Cubic, leaveTangent=-0.274, time=0.6132, value=0.1517}
         <at 1 0>
        ]}}})
    Scale RGB=<1.3 1.3 1.3>
    ScaleA=true
    ScaleRGB=true
   )
   &CameraOffset(Camera Offset Amount=5)
   &SolveForcesAndVelocity(Clamp Velocity=false, Limit Acceleration=false)
  ]
  rendererClasses [&NiagaraSpriteRendererProperties]
  renderers [{rendererClass=&NiagaraSpriteRendererProperties}]
 }
}
summary {
 systemName=NS_Ashton_Milo_Overheat
 emitters [
  {emitterName=MuzzleOverheat, rendererClasses [&NiagaraSpriteRendererProperties]}
  {emitterName=FlashSmoke, rendererClasses [&NiagaraMeshRendererProperties]}
  {bEnabled=false, emitterName=MuzzleOverheat001, simTarget=GPUComputeSim, rendererClasses [&NiagaraSpriteRendererProperties]}
  {bEnabled=false, emitterName=MuzzleOverheat002, simTarget=GPUComputeSim, rendererClasses [&NiagaraSpriteRendererProperties]}
  {bEnabled=false, emitterName=FlashSmoke001, rendererClasses [&NiagaraMeshRendererProperties]}
  {bEnabled=false, emitterName=MuzzleOverheat003, simTarget=GPUComputeSim, rendererClasses [&NiagaraSpriteRendererProperties]}
  {emitterName=MuzzleOverheat004, simTarget=GPUComputeSim, rendererClasses [&NiagaraSpriteRendererProperties]}
  {bEnabled=false, emitterName=MuzzleOverheat005, rendererClasses [&NiagaraSpriteRendererProperties]}
 ]
 userVariables []
}
